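; Base beach ball toy that runs a keep-it-in-the-air minigame and reports the result to squad members.
; The server records player hits after a launch delay, with a per-pawn cooldown map, until the ball hits WorldStatic.
; Ground contact replicates a serial-numbered result struct, whose OnRep prints owner, airtime, hits and other-player count to clients' message feeds.
; When the grounded ball stops, it plays an expiry cue and silently dies.
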(event EventOnPawnTouchedMe
  (params (TouchingPawn :type Pawn%20Object%20Reference) (ImpactNormal :type Vector :default "0, 0, 0") (OtherVelocity :type Vector :default "0, 0, 0"))
  (bind _lastpawnhittimes (Variables|Default|GetLastPawnHitTimes))
  (|CallParentFunction TouchingPawn ImpactNormal OtherVelocity)
  (bind _asfort_player_pawn (Utilities|Casting|CastToFortPlayerPawn TouchingPawn))
  (Networking|SwitchHasAuthority
    (:Authority
      (if (Variables|Default|GetHasTouchedGround)
        _
        (elif (.ReturnValue (Utilities|Map|Find _lastpawnhittimes _asfort_player_pawn))
          (if (>= (- (Utilities|Time|GetGameTimeinSeconds) (Utilities|Map|Find _lastpawnhittimes _asfort_player_pawn)) (Variables|Default|GetTouchCooldownForSinglePawn))
            (CallFunction|RecordPlayerPawnHit _asfort_player_pawn))
          (else
            _))))))

(event Custom|TouchedSomethingStatic
  (Variables|Default|SetTimeInAir (Actor|GetGameTimeSinceCreation))
  (Variables|Default|SetHasTouchedGround true)
  (if (CallFunction|HasMinigameStarted)
    (bind _asfort_player_state (Utilities|Casting|CastToFortPlayerState (Actor|GetOwner))
      (:then
        _)
      (:CastFailed
        _))
    (CallFunction|CreateMessage _asfort_player_state)))

(event EventProjectileStop
  (params (ImpactResult :type Hit%20Result%20Structure%20%28by%20ref%29))
  (|CallParentFunction ImpactResult)
  (Networking|SwitchHasAuthority
    (:Authority
      (if (Variables|Default|GetHasTouchedGround)
        (CallFunction|PlayCueCenteredonMe (Variables|Default|GetCueToPlayWhenExpired))
        (Utilities|FlowControl|Delay 0.01)
        (Gameplay|SilentDie false)))))

(event Collision|EventHit
  (params (MyComp :type Primitive%20Component%20Object%20Reference) (Other :type Actor%20Object%20Reference) (OtherComp :type Primitive%20Component%20Object%20Reference) (bSelfMoved :type Boolean) (HitLocation :type Vector :default "0, 0, 0") (HitNormal :type Vector :default "0, 0, 0") (NormalImpulse :type Vector :default "0, 0, 0") (Hit :type Hit%20Result%20Structure%20%28by%20ref%29))
  (bind (_bblockinghit _binitialoverlap _time _distance _location _impactpoint _normal _impactnormal _physmat _hitactor _hitcomponent) (Collision|BreakHitResult Hit))
  (|CallParentFunction MyComp Other OtherComp bSelfMoved HitLocation HitNormal NormalImpulse Hit)
  (if (not (Variables|Default|GetHasTouchedGround))
    (Networking|SwitchHasAuthority
      (:Authority
        (if _bblockinghit
          (if (Utilities|Enum|Equal(Enum) (Collision|GetCollisionObjectType _hitcomponent) "ECC_WorldStatic")
            (CallFunction|TouchedSomethingStatic)))))))

(event EventSetBallVelocityAndStartTicking
  (params (NewVelocity :type Vector :default "0, 0, 0"))
  (|CallParentFunction NewVelocity)
  (Variables|Default|SetLaunchTime (Utilities|Time|GetGameTimeinSeconds)))

(fn CountOtherPlayers
  (return (- (Utilities|Map|Length (Variables|Default|GetLastPawnHitTimes)) (select (Variables|Default|GetHasBeenTouchedByOwnerPawn) 1 0))))

(fn CreateMessage
  (params (Display%20Message%20Ball%20Owner :type Fort%20Player%20State%20Object%20Reference))
  (if (Variables|/Script/FortniteGame.FortRuntimeOptions|GetbToyMessagingEnabled (Options|GetRuntimeOptions))
    (Variables|SetDisplayMessage_BallOwner_22_EF5D780B4020E911AB429D9BA6B977BA (Utilities|Struct|MakeStructBeachBallMinigameMessage Display%20Message%20Ball%20Owner (Variables|Default|GetNumPawnHits) (Variables|Default|GetTimeInAir) (Math|Conversions|ToByte(Integer) (CallFunction|CountOtherPlayers)) (Variables|Default|GetMessageSN)))
    (Variables|Default|SetMessageSN (Math|Byte|Byte+Byte (Variables|Default|GetMessageSN) 1))))

(fn FormatTimeSpan
  (params (Duration :type Float%20%28double-precision%29))
  (bind _returnvalue (Math|Float|Truncate Duration))
  (if (>= Duration 6e+01)
    (return (Utilities|Text|FormatText "NSLOCTEXT(\"[D451C94A47E4CBCEE9F4209477F83D6D]\", \"42A3A62044196C3BF42F8C85FE1EF776\", \"{MM}:{SS}\")" (Utilities|Text|ToText(Integer) (Math|Float|Truncate (/ Duration 6e+01))) (Utilities|Text|ToText(Integer) _returnvalue false true 2 2)))
    (else
      (return (Utilities|Text|FormatText "NSLOCTEXT(\"[D451C94A47E4CBCEE9F4209477F83D6D]\", \"B74E7DB545619EDE2B823A9EBE020025\", \"{SS} s\")" (Utilities|Text|ToText(Integer) _returnvalue false true 1 2))))))

(fn HasMinigameStarted
  (return (> (Utilities|Map|Length (Variables|Default|GetLastPawnHitTimes)) 0)))

(fn OnRep_DisplayMessage
  (bind _returnvalue (BlueprintContext|GetFortToyContext))
  (bind (_displaymessage_ballowner_22_ef5d780b4020e911ab429d9ba6b977ba _displaymessage_numplayertaps_27_a600d2a048cb01620b914bb4107a6cd5 _displaymessage_timeinair_28_9facd4014a6f7fdaaa915396bd79b184 _displaymessage_numotherplayers_32_7adfeabe4501d2eaaadf789c08762a17) (Utilities|Struct|BreakStructBeachBallMinigameMessage (Variables|GetDisplayMessage_BallOwner_22_EF5D780B4020E911AB429D9BA6B977BA)))
  (Networking|SwitchHasAuthority
    (:Remote
      (Utilities|IsValid _displaymessage_ballowner_22_ef5d780b4020e911ab429d9ba6b977ba
        (:Is%20Valid
          (Utilities|IsValid _returnvalue
            (:Is%20Valid
              (bind _outformatmessage (CallFunction|PickFormatString))
              (bind _returnvalue_2 (Class|FortToyContext|FormatPlayerNameforDisplay (BlueprintContext|GetFortToyContext) _displaymessage_ballowner_22_ef5d780b4020e911ab429d9ba6b977ba "ToyOwner"))
              (bind _timeastext (CallFunction|FormatTimeSpan _displaymessage_timeinair_28_9facd4014a6f7fdaaa915396bd79b184))
              (Class|FortToyContext|ConditionalPrinttoMessageFeed _returnvalue _displaymessage_ballowner_22_ef5d780b4020e911ab429d9ba6b977ba (Transformation|GetActorLocation) "SquadMembersOnly" (Utilities|Text|FormatText _outformatmessage _returnvalue_2 _timeastext _displaymessage_numplayertaps_27_a600d2a048cb01620b914bb4107a6cd5 (Utilities|Text|ToText(Integer) (Math|Conversions|ToInteger(Byte) _displaymessage_numotherplayers_32_7adfeabe4501d2eaaadf789c08762a17) false false))))))))))

(fn PickFormatString
  (bind (_displaymessage_ballowner_22_ef5d780b4020e911ab429d9ba6b977ba _displaymessage_numplayertaps_27_a600d2a048cb01620b914bb4107a6cd5 _displaymessage_timeinair_28_9facd4014a6f7fdaaa915396bd79b184 _displaymessage_numotherplayers_32_7adfeabe4501d2eaaadf789c08762a17) (Utilities|Struct|BreakStructBeachBallMinigameMessage (Variables|GetDisplayMessage_BallOwner_22_EF5D780B4020E911AB429D9BA6B977BA)))
  (return (select (Math|Byte|Byte>Byte _displaymessage_numotherplayers_32_7adfeabe4501d2eaaadf789c08762a17 1) (Variables|Default|GetMessage_PartyWithOthers) (Variables|Default|GetBallInAirMessage))))

(fn RecordPlayerPawnHit
  (params (PawnThatHitMe :type Pawn%20Object%20Reference%20%28by%20ref%29))
  (if (and (> (Variables|Default|GetLaunchTime) 0.0) (> (Utilities|Time|GetGameTimeinSeconds) (+ (Variables|Default|GetLaunchTime) (Variables|Default|GetSpawnCooldownBeforeMinigame))))
    (LocalVariables|SetIsOwnerPawnHit (Utilities|Equal(Object) (Variables|/Script/Engine.Controller|GetPlayerState (Pawn|GetController PawnThatHitMe)) (Actor|GetOwner)))
    (Utilities|Map|Add (Variables|Default|GetLastPawnHitTimes) PawnThatHitMe (Utilities|Time|GetGameTimeinSeconds))
    (Math|Integer|IncrementInt (Variables|Default|GetNumPawnHits)
      (:%20%20
        (if (LocalVariables|GetIsOwnerPawnHit)
          (Variables|Default|SetHasBeenTouchedByOwnerPawn true))))))

(fn ConstructionScript
  (|CallParentFunction))
